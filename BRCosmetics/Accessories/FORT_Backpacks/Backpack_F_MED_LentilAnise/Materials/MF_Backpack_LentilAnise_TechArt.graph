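# Material function for the LentilAnise backpack cosmetic that animates a blinking eyelid over the front and back of its eye.
# MF_Timestamp_InOut_01 turns a timestamp parameter into interpolated blink time and its inverse, alongside clamped FlipBooks that sample two eyelid mask textures at parameterized UVs.
# Three-color lerps tint the eyelids, and sphere masks confine the blends.
# Named reroutes link the stages, and one shared result feeds both Lit and Shaded outputs.
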MaterialFunction /BRCosmetics/Accessories/FORT_Backpacks/Backpack_F_MED_LentilAnise/Materials/MF_Backpack_LentilAnise_TechArt
graph {nodes [
  n1 = ComponentMask(None=NamedRerouteUsage()) {r=true}
  Add(A=@n1, B=ComponentMask(None=NamedRerouteUsage()) {r=true})
  Constant()
  Constant()
  Frac(None=Multiply(A=Time()) {constB=0.5})
  n2 = NamedRerouteUsage()
  FunctionOutput(None=@n2) {outputName=Lit}
  FunctionOutput(None=@n2) {outputName=Shaded, sortPriority=1}
  n3 = VectorParameter() {parameterName=SphereMaskUVs, defaultValue=<rgba 0.8296 0.6056 0 0>}
  n4 = SphereMask(A=TextureCoordinate(), B=AppendVector(A=@n3.R, B=@n3.R), Radius=ScalarParameter() {defaultValue=0.1629, parameterName=Radius}) {attenuationRadius=0.368}
  Multiply(A=NamedRerouteUsage(), B=@n4)
  n5 = TextureCoordinate()
  n6 = VectorParameter() {parameterName=BackOFEyeblinkTextureScaleAndPosition, defaultValue=<rgba 3.0335 3.5 -2.0278 -1.6515>}
  n7 = AppendVector(A=@n6.R, B=@n6.R)
  NamedRerouteDeclaration(None=Add(A=Multiply(A=@n5, B=@n7), B=AppendVector(A=@n6.B, B=@n6.B))) {name=BackOfEyeBlinkUVs, nodeColor=<rgba 0.0383 1 0.2788 1>}
  n8 = TextureObjectParameter() {
   parameterName=BackOfEyelidTexture
   samplerType=SAMPLERTYPE_Masks
   texture=/BRCosmetics/Accessories/FORT_Backpacks/Backpack_F_MED_LentilAnise/Textures/T_LentilAnise_EyeBlinkB
  }
  NamedRerouteDeclaration(None=@n8) {name=BackOfEyelidTexture, nodeColor=<rgba 1 0.0593 0.7851 1>}
  n9 = VertexInterpolator(VS=MaterialFunctionCall(
    TimestampA=ScalarParameter() {parameterName=TimestampEyeBlink}
    TransitionLength_In=ScalarParameter() {parameterName=TransitionLengthEyeBlink}
   ) {materialFunction=&MF_Timestamp_InOut_01}.Linear)
  NamedRerouteDeclaration(None=OneMinus(None=@n9.PS)) {name=InvertedBlinkTime, nodeColor=<rgba 0 0.9766 1 1>}
  n10 = MaterialFunctionCall(
   "Animation  Phase (0-1)"=NamedRerouteUsage()
   Number of Rows=ScalarParameter() {defaultValue=1, parameterName="Number of Rows (S)"}
   Number of Columns=ScalarParameter() {defaultValue=8, parameterName="Number of Columns (S)"}
   Texture=NamedRerouteUsage()
   UVs=NamedRerouteUsage()
   "Clamp Anim (see tooltip)"=StaticBool() {value=true}
  ) {materialFunction=&FlipBook}
  NamedRerouteDeclaration(None=@n10.Alpha) {name=BackOfEyeBlinkAlpha, nodeColor=<rgba 0 1 0.7891 1>}
  NamedRerouteDeclaration(None=@n10.Result) {name=BackOfBlinkTexture, nodeColor=<rgba 1 0.4688 0 1>}
  n11 = VectorParameter() {parameterName=EyelidColorHigh, defaultValue=<rgba 0.7135 0.5222 0.4033 1>}
  n12 = NamedRerouteUsage()
  NamedRerouteDeclaration(None=MaterialFunctionCall(
    A=VectorParameter() {parameterName=EyelidColorDark}.RGB
    B=VectorParameter() {parameterName=EyelidColorMid}.RGB
    C=@n11.RGB
    Alpha=ComponentMask(None=@n12) {r=true}
   ) {materialFunction=&Lerp_3Color}.Result) {name=BackOfEyelidAnimationColor, nodeColor=<rgba 0.0404 0.0967 1 1>}
  n13 = NamedRerouteUsage()
  n14 = VectorParameter() {parameterName=BackOfEyeSphereMaskUVs, defaultValue=<rgba 0.9072 0.3672 0 0>}
  n15 = SphereMask(A=TextureCoordinate(), B=AppendVector(A=@n14.R, B=@n14.R), Radius=ScalarParameter() {defaultValue=0.0853, parameterName=BackOfEyeRadius}) {attenuationRadius=0.368}
  NamedRerouteDeclaration(None=LinearInterpolate(A=NamedRerouteUsage(), B=NamedRerouteUsage(), Alpha=Multiply(A=@n13, B=@n15))) {name=BackOfEyeWithEyelid, nodeColor=<rgba 0.5469 0 1 1>}
  n16 = TextureObjectParameter() {
   parameterName=EyelidTexture
   samplerType=SAMPLERTYPE_Masks
   texture=/BRCosmetics/Accessories/FORT_Backpacks/Backpack_F_MED_LentilAnise/Textures/T_LentilAnise_EyeBlink
  }
  NamedRerouteDeclaration(None=@n16) {name=EyelidTexture, nodeColor=<rgba 0 1 0.0859 1>}
  NamedRerouteDeclaration(None=@n9.PS) {name=BlinkTime, nodeColor=<rgba 0.0274 0.2424 1 1>}
  n17 = TextureCoordinate()
  n18 = VectorParameter() {parameterName=EyeblinkTextureScaleAndPosition, defaultValue=<rgba 3.0335 3.5 -2.0278 -1.6515>}
  n19 = AppendVector(A=@n18.R, B=@n18.R)
  NamedRerouteDeclaration(None=Add(A=Multiply(A=@n17, B=@n19), B=AppendVector(A=@n18.B, B=@n18.B))) {name=EyeBlinkUVs, nodeColor=<rgba 1 0 0.8906 1>}
  n20 = MaterialFunctionCall(
   "Animation  Phase (0-1)"=NamedRerouteUsage()
   Number of Rows=ScalarParameter() {defaultValue=1, parameterName="Number of Rows (S)"}
   Number of Columns=ScalarParameter() {defaultValue=8, parameterName="Number of Columns (S)"}
   Texture=NamedRerouteUsage()
   UVs=NamedRerouteUsage()
   "Clamp Anim (see tooltip)"=StaticBool() {value=true}
  ) {materialFunction=&FlipBook}
  NamedRerouteDeclaration(None=@n20.Alpha) {name=BlinkAlpha, nodeColor=<rgba 0 1 0.7891 1>}
  NamedRerouteDeclaration(None=@n20.Result) {name=BlinkTexture, nodeColor=<rgba 1 0.4688 0 1>}
  NamedRerouteDeclaration(None=FunctionInput(Preview=TextureSample() {texture=/BRCosmetics/Accessories/FORT_Backpacks/Backpack_F_MED_LentilAnise/Textures/T_LentilAnise_Backpack_CL}.RGB) {inputName=Lit}) {name=In_Lit, nodeColor=<rgba 1 0 0.6929 1>}
  NamedRerouteDeclaration(None=FunctionInput() {inputName=Shaded, sortPriority=1}) {name=In_Shaded, nodeColor=<rgba 0.1719 1 0 1>}
  NamedRerouteDeclaration(None=MaterialFunctionCall(
    A=VectorParameter() {parameterName=EyelidColorDark}.RGB
    B=VectorParameter() {parameterName=EyelidColorMid}.RGB
    C=@n11.RGB
    Alpha=Saturate(None=@n1)
   ) {materialFunction=&Lerp_3Color}.Result) {name=EyelidAnimationColor, nodeColor=<rgba 1 0.8672 0 1>}
  n21 = NamedRerouteUsage()
  n22 = Add(A=@n4, B=@n15)
  NamedRerouteDeclaration(None=LinearInterpolate(A=NamedRerouteUsage(), B=NamedRerouteUsage(), Alpha=Multiply(A=@n21, B=@n22))) {name=EyeWithEyelid, nodeColor=<rgba 0.5469 0 1 1>}
  NamedRerouteUsage()
  NamedRerouteUsage()
  ScalarParameter() {defaultValue=0.6588}
  VectorParameter() {parameterName=EyelidColorHigh, defaultValue=<rgba 0.7292 0.596 0.4882 1>}
 ]}
