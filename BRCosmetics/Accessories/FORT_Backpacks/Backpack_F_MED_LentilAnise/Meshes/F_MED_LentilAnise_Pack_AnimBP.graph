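# Animation Blueprint for the LentilAnise backpack, animating eight scale bones and an eye.
# Self-rescheduling random timers pick eye look targets and an open/closed pulse level, centering the eye when low.
# Each update eases both values and blends phase-offset pulse waves into per-scale transform arrays.
# The AnimGraph chains additive bone modifiers, ending in a bool-gated pack offset; a flag records whether the matching LentilAnise body is worn.
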AnimBlueprint /BRCosmetics/Accessories/FORT_Backpacks/Backpack_F_MED_LentilAnise/Meshes/F_MED_LentilAnise_Pack_AnimBP
target_skeleton=&/F_MED_LentilAnise_Pack_Skeleton
graphs [
 AnimGraph {nodes [
   n1 = "Transform (Modify) Bone - Bone: scales_01"(ComponentPose=Local To Component(LocalPose=Input Pose()), Alpha=1, Translation="0,0,0", Rotation="0,0,0", Scale="1,1,1") {
    rotationMode=BMM_Additive
    rotationSpace=BCS_BoneSpace
    translationMode=BMM_Additive
    translationSpace=BCS_BoneSpace
    boneToModify {boneName=scales_01}
   }
   n2 = "Transform (Modify) Bone - Bone: scales_02"(ComponentPose=@n1, Alpha=1, Translation="0,0,0", Rotation="0,0,0", Scale="1,1,1") {
    rotationMode=BMM_Additive
    rotationSpace=BCS_BoneSpace
    translationMode=BMM_Additive
    translationSpace=BCS_BoneSpace
    boneToModify {boneName=scales_02}
   }
   n3 = "Transform (Modify) Bone - Bone: scales_03"(ComponentPose=@n2, Alpha=1, Translation="0,0,0", Rotation="0,0,0", Scale="1,1,1") {
    rotationMode=BMM_Additive
    rotationSpace=BCS_BoneSpace
    translationMode=BMM_Additive
    translationSpace=BCS_BoneSpace
    boneToModify {boneName=scales_03}
   }
   n4 = "Transform (Modify) Bone - Bone: scales_04"(ComponentPose=@n3, Alpha=1, Translation="0,0,0", Rotation="0,0,0", Scale="1,1,1") {
    rotationMode=BMM_Additive
    rotationSpace=BCS_BoneSpace
    translationMode=BMM_Additive
    translationSpace=BCS_BoneSpace
    boneToModify {boneName=scales_04}
   }
   n5 = "Transform (Modify) Bone - Bone: scales_05"(ComponentPose=@n4, Alpha=1, Translation="0,0,0", Rotation="0,0,0", Scale="1,1,1") {
    rotationMode=BMM_Additive
    rotationSpace=BCS_BoneSpace
    translationMode=BMM_Additive
    translationSpace=BCS_BoneSpace
    boneToModify {boneName=scales_05}
   }
   n6 = "Transform (Modify) Bone - Bone: scales_06"(ComponentPose=@n5, Alpha=1, Translation="0,0,0", Rotation="0,0,0", Scale="1,1,1") {
    rotationMode=BMM_Additive
    rotationSpace=BCS_BoneSpace
    translationMode=BMM_Additive
    translationSpace=BCS_BoneSpace
    boneToModify {boneName=scales_06}
   }
   n7 = "Transform (Modify) Bone - Bone: scales_07"(ComponentPose=@n6, Alpha=1, Translation="0,0,0", Rotation="0,0,0", Scale="1,1,1") {
    rotationMode=BMM_Additive
    rotationSpace=BCS_BoneSpace
    translationMode=BMM_Additive
    translationSpace=BCS_BoneSpace
    boneToModify {boneName=scales_07}
   }
   n8 = "Transform (Modify) Bone - Bone: scales_08"(ComponentPose=@n7, Alpha=1, Translation="0,0,0", Rotation="0,0,0", Scale="1,1,1") {
    rotationMode=BMM_Additive
    rotationSpace=BCS_BoneSpace
    translationMode=BMM_Additive
    translationSpace=BCS_BoneSpace
    boneToModify {boneName=scales_08}
   }
   n9 = "Transform (Modify) Bone - Bone: Middle"(
    ComponentPose="Transform (Modify) Bone - Bone: eye"(ComponentPose=@n8, Alpha=1, Translation="0,0,0", Rotation="0,0,0", Scale="1,1,1") {rotationMode=BMM_Additive, boneToModify {boneName=eye}}
    Alpha=1
    Translation="0,1,0"
    Rotation="0,0,0"
    Scale="1,1,1"
   ) {translationMode=BMM_Additive, translationSpace=BCS_ParentBoneSpace, boneToModify {boneName=Middle}, translation {y=1}}
   n10 = "Transform (Modify) Bone - Bone: pack"!(ComponentPose=@n9, bAlphaBoolEnabled:bool, Translation="0,1.5,0", Rotation="0,0,6", Scale) -> (Pose) {
    alphaInputType=Bool
    rotationMode=BMM_Additive
    rotationSpace=BCS_ParentBoneSpace
    translationMode=BMM_Additive
    translationSpace=BCS_ParentBoneSpace
    boneToModify {boneName=pack}
    rotation {roll=6}
    translation {y=1.5}
   }
   Output Pose(Result=Component To Local(ComponentPose=@n10)) {layerGroup=None}
  ]}
 EventGraph {nodes [
   eye target change() {is=title K2Node_CustomEvent}
   > Set Eye target Rot(Eye target Rot:Rotator="Lerp (Rotator)"!(
     A:Rotator="0, 0, 0"
     B_Roll:float=RandomFloatInRange(Min=-35, Max=35)
     B_Pitch:float=0
     B_Yaw:float=RandomFloatInRange(Min=-40, Max=40)
     Alpha:float=Get Puls Alpha() -> (Puls Alpha:double)
     bShortestPath:bool=false
    ) -> (ReturnValue:Rotator)) -> (Output_Get:Rotator)
   > Set Eye target interpolation speed(Eye target interpolation speed:double=RandomFloatInRange(Min=12, Max=24)) -> (Output_Get:double)
   > Set Timer by Function Name!(
    execute
    Object:Object Reference
    FunctionName:String=eye target change
    Time:float=RandomFloatInRange(Min=0.3, Max=2)
    bLooping:bool=false
    bMaxOncePerFrame:bool=false
    InitialStartDelay:float=0
    InitialStartDelayVariance:float=0
   ) -> (then, ReturnValue:Timer Handle Structure) {is=title K2Node_CallFunction}
   close down() {is=title K2Node_CustomEvent}
   > n1 = Set Pulse Alpha target(Pulse Alpha target:double="Clamp (Float)"(Value=RandomFloatInRange(Min=-0.2, Max=1.5), Min=0, Max=1)) -> (Output_Get:double)
   > Set Timer by Function Name(
    FunctionName=close down
    Time=Lerp(A=1, B=3, Alpha=@n1)
    bLooping=false
    bMaxOncePerFrame=false
    InitialStartDelay=0
    InitialStartDelayVariance=0
   )
   > Branch(Condition="float < float"(A=Get Pulse Alpha target() -> (Pulse Alpha target:double), B=0.4))
   > Set Eye target Rot(Eye target Rot="0, 0, 0")
   > Set Eye target interpolation speed(Eye target interpolation speed=12)
   Event BlueprintUpdateAnimation()
   > n2 = Sequence!(execute) -> (then_0, then_1, then_2:Exec)
   > n3 = Scales Sliding!(
    execute
    self
    Puls Alpha:double=Get Puls Alpha()
    TimeOffset:double=0.125
    puls speed:double=1
    puls zer position:double=-1
    Exp puls:double=1
   ) -> (then, Alpha array:"Array of Float (double-precision)s")
   > n4 = For Each Loop(Array:"Array of Float (double-precision)s"=@n3) -> (Array Element:double)
   n5 = float * float(A=@n4, B=-4) {is=title K2Node_PromotableOperator}
   > Set Array Elem!(
    execute
    TargetArray:Array of Rotators=Get Scales_Rot() -> (Scales_Rot:Array of Rotators)
    Index:int=@n4.Array Index
    Item:"Rotator (by ref)"="Lerp (Rotator)"!(A="3,0,0", B:Rotator=MakeRotator(Pitch=@n5), Alpha=Get Puls Alpha(), bShortestPath=false) -> (ReturnValue)
    bSizeToFit:bool=false
   ) -> (then)
   n6 = float * float(A=@n4, B=0.2)
   > Set Array Elem(
    TargetArray:Array of Vectors=Get Scales_Loc() -> (Scales_Loc:Array of Vectors)
    Index=@n4.Array Index
    Item:"Vector (by ref)"="Lerp (Vector)"!(A:Vector="-2.5,0,1", B:Vector=MakeVector(X=@n6, Y=0), Alpha:float=Get Puls Alpha()) -> (ReturnValue:Vector)
    bSizeToFit=false
   )
   Set Eye_rot(
    execute=@n2.then_1
    Eye_rot:Rotator=RInterpTo!(
     Current:Rotator=Get Eye_rot() -> (Eye_rot:Rotator)
     Target:Rotator=Get Eye target Rot() -> (Eye target Rot:Rotator)
     DeltaTime:float=GetWorldDeltaSeconds!() -> (ReturnValue:double)
     InterpSpeed:float=Get Eye target interpolation speed() -> (Eye target interpolation speed:double)
    ) -> (ReturnValue:Rotator)
   ) -> (Output_Get:Rotator)
   > Set Puls Alpha(Puls Alpha:double=FInterpTo(Current=Get Puls Alpha(), Target=Get Pulse Alpha target(), DeltaTime=GetWorldDeltaSeconds(), InterpSpeed=8)) -> (Output_Get:double)
   Event BlueprintInitializeAnimation()
   > eye target change() {is=title K2Node_CallFunction}
   > close down() {is=title K2Node_CallFunction}
   n7 = Event OnCharacterCustomizationLoaded()
   > Set Char is LentilAnise?(Char is LentilAnise?:bool=Contains Item(
     TargetArray:Array of Custom Character Part Object References=@n7.CustomParts
     ItemToFind:"Custom Character Part Object Reference (by ref)"=/BRCosmetics/Athena/Heroes/Meshes/Bodies/CP_Athena_Body_F_LentilAnise
    )) -> (Output_Get:bool)
   TryGetPawnOwner()
  ]}
 Scales Sliding {nodes [
   n1 = Scales Sliding!() -> (then, Puls Alpha:double=1, TimeOffset:double=2, puls speed:double=3, puls zer position:double=1, Exp puls:double=1) {is=title K2Node_FunctionEntry}
   > n2 = For Loop(LastIndex=7)
   Return Node!(execute=@n2.Completed, Alpha array:"Array of Float (double-precision)s"=Get Alphas() -> (Alphas:"Array of Float (double-precision)s"))
   n3 = MakePulsatingValue!(
    InCurrentTime:float=GetGameTimeInSeconds()
    InPulsesPerSecond:float=@n1.puls speed
    InPhase:float=float * float!(A:int=@n2, B=@n1.TimeOffset, C:double=@n1) -> (ReturnValue)
   ) -> (ReturnValue:float)
   > Set Array Elem(
    TargetArray:"Array of Float (double-precision)s"=Get Alphas()
    Index=@n2
    Item:"Float (double-precision) (by ref)"=float + float(A=Power!(Base:double=@n3, Exp:double=@n1.Exp puls) -> (ReturnValue:double) {is=title K2Node_CallFunction}, B=@n1.puls zer position)
    bSizeToFit=true
   )
  ]}
]
variables [
 Char is LentilAnise? {type=boolean}
 Eye target interpolation speed {type=number}
 Eye target Rot {type=Rotator, default=<rot 32 0 32>}
 Eye_rot {type=Rotator}
 Puls Alpha {type=number, default=1}
 Pulse Alpha target {type=number}
 Scales_Loc {type="array<Vector>", default [<0 0 0>, <0 0 0>, <0 0 0>, <0 0 0>, <0 0 0>, <0 0 0>, <0 0 0>, <0 0 0>]}
 Scales_Rot {type="array<Rotator>", default [<rot 0 0 0>, <rot 0 0 0>, <rot 0 0 0>, <rot 0 0 0>, <rot 0 0 0>, <rot 0 0 0>, <rot 0 0 0>, <rot 0 0 0>]}
]
